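# Physics asset for the Sailor Squad Rose one-handed female pickaxe skeletal mesh, defining collision and simulated motion for its dangling chain.
# A capsule covers the handle bone, and small capsules sit on a master chain bone and three dynamic chain bones.
# Constraints link the chain bones in series from the master, with limited swing and twist and a looser middle joint; none connects the chain to the handle body.
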PhysicsAsset /BRCosmetics/Weapons/FORT_Melee/Pickaxe_Sailor_Squad_Rose_Female_1H/Meshes/Sailor_Squad_Rose_1h_Female_Axe_Physics
bodies {
 Melee_TwoHanded_Handle [{shapeType=Capsule, length=28.7531, radius=6.0381, center=<0 0 11.5415>}]
 dyn_chain_01 [{shapeType=Capsule, length=0.8386, radius=0.8174, center=<1.3285 0 0.0001>, rotation=<rot 90 0 0>}]
 dyn_chain_02 [{shapeType=Capsule, length=0.0004, radius=2, center=<1.9985 0 0>, rotation=<rot 90 0 0>}]
 dyn_chain_03 [{shapeType=Capsule, length=0.0001, radius=2.5594, center=<2.4301 0 0>, rotation=<rot 90 0 0>}]
 dyn_chain_master [{shapeType=Capsule, length=2, radius=2, rotation=<rot 90 0 0>}]
}
constraints [
 {bone1Name=dyn_chain_01, bone2Name=dyn_chain_master, swing1LimitDegrees=45, swing2LimitDegrees=25, twistLimitDegrees=5}
 {bone1Name=dyn_chain_02, bone2Name=dyn_chain_01, swing1LimitDegrees=45, swing2LimitDegrees=45, twistLimitDegrees=45}
 {bone1Name=dyn_chain_03, bone2Name=dyn_chain_02, swing1LimitDegrees=45, swing2LimitDegrees=25, twistLimitDegrees=5}
]
